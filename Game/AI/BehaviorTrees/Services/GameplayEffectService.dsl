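; Behavior tree service that applies or removes a gameplay effect on a blackboard actor when it activates and deactivates, by default applying then removing on self.
; Activation and deactivation each pass their own effect, key, tag queries, behavior and stack count into a shared HandleGE function.
; HandleGE checks optional tag queries against the AI's pawn and the target, then uses the target's Fort ability system component as both source and target.
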(event EventReceiveTick
  (params (OwnerActor :type Actor%20Object%20Reference) (DeltaSeconds :type Float%20%28single-precision%29)))

(event EventReceiveActivation
  (params (OwnerActor :type Actor%20Object%20Reference))
  ;[1 1 1] Manage application and removal of GEs.
  ;
  ; Default behavior is to Apply GE to Self on Activation and Remove GE from Self on Deactivation.
  (CallFunction|HandleGE OwnerActor (Variables|Activation|GetActivationOwnerTagQuery) (Variables|Activation|GetActivationGameplayEffect) (Variables|Activation|GetActivationEffectBlackboardKey) (Variables|Activation|GetActivationBlackboardKeyTagQuery) (Variables|Activation|GetActivationBehavior) (Variables|Activation|GetActivationEffectStacksToRemove)))

(event EventReceiveDeactivation
  (params (OwnerActor :type Actor%20Object%20Reference))
  ;[1 1 1] Manage application and removal of GEs.
  ;
  ; Default behavior is to Apply GE to Self on Activation and Remove GE from Self on Deactivation.
  (CallFunction|HandleGE OwnerActor (Variables|Deactivation|GetDeactivationOwnerTagQuery) (Variables|Deactivation|GetDeactivationGameplayEffect) (Variables|Deactivation|GetDeactivationEffectBlackboardKey) (Variables|Deactivation|GetDeactivationBlackboardKeyTagQuery) (Variables|Deactivation|GetDeactivationBehavior) (Variables|Deactivation|GetDeactivationEffectStacksToRemove)))

(fn HandleGE
  (params (OwnerActor :type Actor%20Object%20Reference) (Owner%20TagQuery :type Gameplay%20Tag%20Query%20Structure) (GE%20Class :type Gameplay%20Effect%20Class%20Reference) (BB%20Key :type Blackboard%20Key%20Selector%20Structure%20%28by%20ref%29) (BB%20TagQuery :type Gameplay%20Tag%20Query%20Structure%20%28by%20ref%29) (GE%20Behavior :type Enum_GameplayEffectService_EventBehavior%20Enum :default "NewEnumerator0") (GE%20RemoveStacks :type Integer))
  (bind _returnvalue (AI|BehaviorTree|GetBlackboardValueasActor BB%20Key))
  (bind _returnvalue_2 (Ability|GetFortAbilitySystemComponent _returnvalue))
  (if (not (GameplayTags|IsTagQueryEmpty Owner%20TagQuery))
    (bind _asaicontroller (Utilities|Casting|CastToAIController OwnerActor))
    (Utilities|IsValid (Pawn|GetControlledPawn _asaicontroller)
      (:Is%20Valid
        (Utilities|IsValid (Ability|GetFortAbilitySystemComponent (Pawn|GetControlledPawn _asaicontroller))
          (:Is%20Valid
            (if (GameplayTags|DoesContainerMatchTagQuery (GameplayTags|GetOwnedGameplayTags (Ability|GetFortAbilitySystemComponent (Pawn|GetControlledPawn _asaicontroller)) :!class /Script/GameplayTags.GameplayTagAssetInterface) Owner%20TagQuery)
              _)))))
    (else
      _))
  (if (Utilities|IsValidClass GE%20Class)
    (Utilities|IsValid _returnvalue :!pinned true
      (:Is%20Valid
        (Utilities|IsValid _returnvalue_2 :!pinned true
          (:Is%20Valid
            (if (not (GameplayTags|IsTagQueryEmpty BB%20TagQuery))
              (if (GameplayTags|DoesContainerMatchTagQuery (GameplayTags|GetOwnedGameplayTags _returnvalue_2 :!class /Script/GameplayTags.GameplayTagAssetInterface) BB%20TagQuery)
                _)
              (else
                _))
            (switch Utilities|FlowControl|Switch|SwitchonEnum_GameplayEffectService_EventBehavior GE%20Behavior
              (:NewEnumerator1
                (GameplayEffects|ApplyGameplayEffectToTarget :self _returnvalue_2 :GameplayEffectClass GE%20Class :Target _returnvalue_2 :Level 0.0 :!pinned true) ; Apply GE
              )
              (:NewEnumerator2
                (GameplayEffects|RemoveActiveGameplayEffectbySourceEffect _returnvalue_2 GE%20Class _returnvalue_2 GE%20RemoveStacks :!pinned true) ; Remove GE
              )))) ; Actor must have Ability System Component
      )) ; BB Key must be an Actor
  ))
